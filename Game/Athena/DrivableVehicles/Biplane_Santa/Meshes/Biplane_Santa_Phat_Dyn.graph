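# Physics asset for the Santa biplane's dangling decorations, giving antlers, bells and wing lights bodies for secondary motion.
# Fuselage shapes and four small wing capsules act as anchors; the wing-to-fuselage joints leave swing and twist free.
# Antlers and two short bell chains hang from the fuselage with tight angular limits.
# Each light capsule joins directly to the fuselage (low-numbered) or its wing, not chained; right-side light bones are named "dyn_rights".
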PhysicsAsset /Game/Athena/DrivableVehicles/Biplane_Santa/Meshes/Biplane_Santa_Phat_Dyn
bodies {
 dyn_antler_l [{shapeType=Capsule, length=12.9721, radius=4.8711, center=<0 -0.4 0>, rotation=<rot 0 0 90>}]
 dyn_antler_r [{shapeType=Capsule, length=12.9721, radius=4.8711, center=<0 0.4 0>, rotation=<rot 0 0 90>}]
 dyn_bells_01 [{shapeType=Capsule, length=6.0939, radius=2, center=<5 0 0>, rotation=<rot 90 0 0>}]
 dyn_bells_02 [{shapeType=Capsule, length=6.0939, radius=2, center=<5 0 0>, rotation=<rot 90 0 0>}]
 dyn_bells_03 [{shapeType=Box, extentX=7.5, extentY=5.4992, extentZ=2.0402, center=<2.0359 -0.033 4.7651>, rotation=<rot 66.9973 1.7496 8.1316>}]
 dyn_bells_fr_01 [{shapeType=Capsule, length=3.7631, radius=2, center=<2 0 0>, rotation=<rot 90 0 0>}]
 dyn_bells_fr_02 [{shapeType=Capsule, length=3.7631, radius=2, center=<2 0 0>, rotation=<rot 90 0 0>}]
 dyn_lights_bott_l_01 [{shapeType=Capsule, length=7.6446, radius=3.4089, center=<0 0 0.6>}]
 dyn_lights_bott_l_02 [{shapeType=Capsule, length=7.6446, radius=3.3053, center=<0 0 0.6>}]
 dyn_lights_bott_l_03 [{shapeType=Capsule, length=7.6446, radius=3.6, center=<0 0 0.6>}]
 dyn_lights_bott_l_04 [{shapeType=Capsule, length=7.6446, radius=3.3355, center=<0 0 0.6>}]
 dyn_lights_bott_l_05 [{shapeType=Capsule, length=7.6446, radius=3.342, center=<0 0 0.6>}]
 dyn_lights_bott_l_06 [{shapeType=Capsule, length=7.6446, radius=3.3834, center=<0 0 0.6>}]
 dyn_lights_bott_l_07 [{shapeType=Capsule, length=7.6446, radius=3.3074, center=<0 0 0.6>}]
 dyn_lights_bott_l_08 [{shapeType=Capsule, length=7.6446, radius=3.3946, center=<0 0 0.6>}]
 dyn_lights_bott_l_09 [{shapeType=Capsule, length=7.6446, radius=3.4269, center=<0 0 0.6>}]
 dyn_lights_top_l_01 [{shapeType=Capsule, length=7.6446, radius=3.611, center=<0 0 0.6>}]
 dyn_lights_top_l_010 [{shapeType=Capsule, length=7.6446, radius=3.5435, center=<0 0 0.6>}]
 dyn_lights_top_l_011 [{shapeType=Capsule, length=7.6446, radius=3.4086, center=<0 0 0.6>}]
 dyn_lights_top_l_012 [{shapeType=Capsule, length=7.6446, radius=3.3694, center=<0 0 0.6>}]
 dyn_lights_top_l_013 [{shapeType=Capsule, length=7.6446, radius=3.3766, center=<0 0 0.6>}]
 dyn_lights_top_l_02 [{shapeType=Capsule, length=7.6347, radius=3.3768, center=<0 0 0.6>}]
 dyn_lights_top_l_03 [{shapeType=Capsule, length=7.6446, radius=3.4088, center=<0 0 0.6>}]
 dyn_lights_top_l_04 [{shapeType=Capsule, length=7.6446, radius=3.2783, center=<0 0 0.6>}]
 dyn_lights_top_l_05 [{shapeType=Capsule, length=7.6446, radius=3.3144, center=<0 0 0.6>}]
 dyn_lights_top_l_06 [{shapeType=Capsule, length=7.6446, radius=3.3181, center=<0 0 0.6>}]
 dyn_lights_top_l_07 [{shapeType=Capsule, length=7.6446, radius=3.4089, center=<0 0 0.6>}]
 dyn_lights_top_l_08 [{shapeType=Capsule, length=7.6446, radius=3.3766, center=<0 0 0.6>}]
 dyn_lights_top_l_09 [{shapeType=Capsule, length=7.6446, radius=3.3815, center=<0 0 0.6>}]
 dyn_rights_bott_r_01 [{shapeType=Capsule, length=7.6446, radius=3.4089, center=<0 0 0.6>}]
 dyn_rights_bott_r_02 [{shapeType=Capsule, length=7.6446, radius=3.3053, center=<0 0 0.6>}]
 dyn_rights_bott_r_03 [{shapeType=Capsule, length=7.6446, radius=3.6, center=<0 0 0.6>}]
 dyn_rights_bott_r_04 [{shapeType=Capsule, length=7.6446, radius=3.3355, center=<0 0 0.6>}]
 dyn_rights_bott_r_05 [{shapeType=Capsule, length=7.6446, radius=3.342, center=<0 0 0.6>}]
 dyn_rights_bott_r_06 [{shapeType=Capsule, length=7.6446, radius=3.3834, center=<0 0 0.6>}]
 dyn_rights_bott_r_07 [{shapeType=Capsule, length=7.6446, radius=3.3074, center=<0 0 0.6>}]
 dyn_rights_bott_r_08 [{shapeType=Capsule, length=7.6446, radius=3.3946, center=<0 0 0.6>}]
 dyn_rights_bott_r_09 [{shapeType=Capsule, length=7.6446, radius=3.4269, center=<0 0 0.6>}]
 dyn_rights_top_r_01 [{shapeType=Capsule, length=7.6446, radius=3.4484, center=<0 0 0.6>}]
 dyn_rights_top_r_010 [{shapeType=Capsule, length=7.6446, radius=3.5435, center=<0 0 0.6>}]
 dyn_rights_top_r_011 [{shapeType=Capsule, length=7.6446, radius=3.4086, center=<0 0 0.6>}]
 dyn_rights_top_r_012 [{shapeType=Capsule, length=7.6446, radius=3.3694, center=<0 0 0.6>}]
 dyn_rights_top_r_013 [{shapeType=Capsule, length=7.6446, radius=3.3766, center=<0 0 0.6>}]
 dyn_rights_top_r_02 [{shapeType=Capsule, length=7.6446, radius=3.4081, center=<0 0 0.6>}]
 dyn_rights_top_r_03 [{shapeType=Capsule, length=7.6446, radius=3.4088, center=<0 0 0.6>}]
 dyn_rights_top_r_04 [{shapeType=Capsule, length=7.6446, radius=3.2783, center=<0 0 0.6>}]
 dyn_rights_top_r_05 [{shapeType=Capsule, length=7.6446, radius=3.3144, center=<0 0 0.6>}]
 dyn_rights_top_r_06 [{shapeType=Capsule, length=7.6446, radius=3.3181, center=<0 0 0.6>}]
 dyn_rights_top_r_07 [{shapeType=Capsule, length=7.6446, radius=3.4089, center=<0 0 0.6>}]
 dyn_rights_top_r_08 [{shapeType=Capsule, length=7.6446, radius=3.3766, center=<0 0 0.6>}]
 dyn_rights_top_r_09 [{shapeType=Capsule, length=7.6446, radius=3.3814, center=<0 0 0.6>}]
 fuselage [
  {shapeType=Capsule, length=25.0108, radius=6.8897, center=<-81.9217 -12.4254 54.8181>}
  {shapeType=Box, extentX=26.6222, extentY=43.009, extentZ=7.5, center=<-78.3728 0 40.3151>}
  {shapeType=Box, extentX=40.0465, extentY=48.7628, extentZ=7.5, center=<-61.1087 0 46.6451>, rotation=<rot 24.929 0 0>}
 ]
 wing_bott_l [{shapeType=Capsule, length=5, radius=5}]
 wing_bott_r [{shapeType=Capsule, length=5, radius=5}]
 wing_top_l [{shapeType=Capsule, length=5, radius=5}]
 wing_top_r [{shapeType=Capsule, length=5, radius=5}]
}
constraints [
 {bone1Name=dyn_bells_03, bone2Name=dyn_bells_02, swing1LimitDegrees=20, swing2LimitDegrees=20, twistLimitDegrees=5}
 {bone1Name=dyn_bells_02, bone2Name=dyn_bells_01, swing1LimitDegrees=20, swing2LimitDegrees=20, twistLimitDegrees=5}
 {bone1Name=dyn_bells_fr_02, bone2Name=dyn_bells_fr_01, swing1LimitDegrees=2, swing2LimitDegrees=20, twistLimitDegrees=5}
 {bone1Name=dyn_antler_l, bone2Name=fuselage, swing1LimitDegrees=10, swing2LimitDegrees=10, twistLimitDegrees=10}
 {bone1Name=dyn_bells_01, bone2Name=fuselage, swing1LimitDegrees=20, swing2LimitDegrees=20, twistLimitDegrees=5}
 {bone1Name=dyn_antler_r, bone2Name=fuselage, swing1LimitDegrees=10, swing2LimitDegrees=10, twistLimitDegrees=10}
 {bone1Name=dyn_bells_fr_01, bone2Name=fuselage, swing1LimitDegrees=20, swing2LimitDegrees=20, twistLimitDegrees=5}
 {
  bone1Name=wing_top_l
  bone2Name=fuselage
  swing1LimitDegrees=45
  swing1Motion=Free
  swing2LimitDegrees=45
  swing2Motion=Free
  twistLimitDegrees=45
  twistMotion=Free
 }
 {
  bone1Name=wing_bott_l
  bone2Name=fuselage
  swing1LimitDegrees=45
  swing1Motion=Free
  swing2LimitDegrees=45
  swing2Motion=Free
  twistLimitDegrees=45
  twistMotion=Free
 }
 {
  bone1Name=wing_top_r
  bone2Name=fuselage
  swing1LimitDegrees=45
  swing1Motion=Free
  swing2LimitDegrees=45
  swing2Motion=Free
  twistLimitDegrees=45
  twistMotion=Free
 }
 {
  bone1Name=wing_bott_r
  bone2Name=fuselage
  swing1LimitDegrees=45
  swing1Motion=Free
  swing2LimitDegrees=45
  swing2Motion=Free
  twistLimitDegrees=45
  twistMotion=Free
 }
 {bone1Name=dyn_lights_top_l_06, bone2Name=wing_top_l, swing1LimitDegrees=45, swing2LimitDegrees=45, twistLimitDegrees=45}
 {bone1Name=dyn_lights_top_l_07, bone2Name=wing_top_l, swing1LimitDegrees=45, swing2LimitDegrees=45, twistLimitDegrees=45}
 {bone1Name=dyn_lights_top_l_08, bone2Name=wing_top_l, swing1LimitDegrees=45, swing2LimitDegrees=45, twistLimitDegrees=45}
 {bone1Name=dyn_lights_top_l_09, bone2Name=wing_top_l, swing1LimitDegrees=45, swing2LimitDegrees=45, twistLimitDegrees=45}
 {bone1Name=dyn_lights_top_l_010, bone2Name=wing_top_l, swing1LimitDegrees=45, swing2LimitDegrees=45, twistLimitDegrees=45}
 {bone1Name=dyn_lights_top_l_011, bone2Name=wing_top_l, swing1LimitDegrees=45, swing2LimitDegrees=45, twistLimitDegrees=45}
 {bone1Name=dyn_lights_top_l_012, bone2Name=wing_top_l, swing1LimitDegrees=45, swing2LimitDegrees=45, twistLimitDegrees=45}
 {bone1Name=dyn_lights_top_l_013, bone2Name=wing_top_l, swing1LimitDegrees=45, swing2LimitDegrees=45, twistLimitDegrees=45}
 {bone1Name=dyn_lights_bott_l_03, bone2Name=wing_bott_l, swing1LimitDegrees=45, swing2LimitDegrees=45, twistLimitDegrees=45}
 {bone1Name=dyn_lights_bott_l_04, bone2Name=wing_bott_l, swing1LimitDegrees=45, swing2LimitDegrees=45, twistLimitDegrees=45}
 {bone1Name=dyn_lights_bott_l_05, bone2Name=wing_bott_l, swing1LimitDegrees=45, swing2LimitDegrees=45, twistLimitDegrees=45}
 {bone1Name=dyn_lights_bott_l_06, bone2Name=wing_bott_l, swing1LimitDegrees=45, swing2LimitDegrees=45, twistLimitDegrees=45}
 {bone1Name=dyn_lights_bott_l_07, bone2Name=wing_bott_l, swing1LimitDegrees=45, swing2LimitDegrees=45, twistLimitDegrees=45}
 {bone1Name=dyn_lights_bott_l_08, bone2Name=wing_bott_l, swing1LimitDegrees=45, swing2LimitDegrees=45, twistLimitDegrees=45}
 {bone1Name=dyn_lights_bott_l_09, bone2Name=wing_bott_l, swing1LimitDegrees=45, swing2LimitDegrees=45, twistLimitDegrees=45}
 {bone1Name=dyn_rights_top_r_06, bone2Name=wing_top_r, swing1LimitDegrees=45, swing2LimitDegrees=45, twistLimitDegrees=45}
 {bone1Name=dyn_rights_top_r_07, bone2Name=wing_top_r, swing1LimitDegrees=45, swing2LimitDegrees=45, twistLimitDegrees=45}
 {bone1Name=dyn_rights_top_r_08, bone2Name=wing_top_r, swing1LimitDegrees=45, swing2LimitDegrees=45, twistLimitDegrees=45}
 {bone1Name=dyn_rights_top_r_09, bone2Name=wing_top_r, swing1LimitDegrees=45, swing2LimitDegrees=45, twistLimitDegrees=45}
 {bone1Name=dyn_rights_top_r_010, bone2Name=wing_top_r, swing1LimitDegrees=45, swing2LimitDegrees=45, twistLimitDegrees=45}
 {bone1Name=dyn_rights_top_r_011, bone2Name=wing_top_r, swing1LimitDegrees=45, swing2LimitDegrees=45, twistLimitDegrees=45}
 {bone1Name=dyn_rights_top_r_012, bone2Name=wing_top_r, swing1LimitDegrees=45, swing2LimitDegrees=45, twistLimitDegrees=45}
 {bone1Name=dyn_rights_top_r_013, bone2Name=wing_top_r, swing1LimitDegrees=45, swing2LimitDegrees=45, twistLimitDegrees=45}
 {bone1Name=dyn_rights_bott_r_04, bone2Name=wing_bott_r, swing1LimitDegrees=45, swing2LimitDegrees=45, twistLimitDegrees=45}
 {bone1Name=dyn_rights_bott_r_03, bone2Name=wing_bott_r, swing1LimitDegrees=45, swing2LimitDegrees=45, twistLimitDegrees=45}
 {bone1Name=dyn_rights_bott_r_05, bone2Name=wing_bott_r, swing1LimitDegrees=45, swing2LimitDegrees=45, twistLimitDegrees=45}
 {bone1Name=dyn_rights_bott_r_06, bone2Name=wing_bott_r, swing1LimitDegrees=45, swing2LimitDegrees=45, twistLimitDegrees=45}
 {bone1Name=dyn_rights_bott_r_07, bone2Name=wing_bott_r, swing1LimitDegrees=45, swing2LimitDegrees=45, twistLimitDegrees=45}
 {bone1Name=dyn_rights_bott_r_08, bone2Name=wing_bott_r, swing1LimitDegrees=45, swing2LimitDegrees=45, twistLimitDegrees=45}
 {bone1Name=dyn_rights_bott_r_09, bone2Name=wing_bott_r, swing1LimitDegrees=45, swing2LimitDegrees=45, twistLimitDegrees=45}
 {bone1Name=dyn_lights_top_l_01, bone2Name=fuselage, swing1LimitDegrees=45, swing2LimitDegrees=45, twistLimitDegrees=45}
 {bone1Name=dyn_lights_top_l_02, bone2Name=fuselage, swing1LimitDegrees=45, swing2LimitDegrees=45, twistLimitDegrees=45}
 {bone1Name=dyn_lights_top_l_03, bone2Name=fuselage, swing1LimitDegrees=45, swing2LimitDegrees=45, twistLimitDegrees=45}
 {bone1Name=dyn_lights_top_l_04, bone2Name=fuselage, swing1LimitDegrees=45, swing2LimitDegrees=45, twistLimitDegrees=45}
 {bone1Name=dyn_lights_top_l_05, bone2Name=fuselage, swing1LimitDegrees=45, swing2LimitDegrees=45, twistLimitDegrees=45}
 {bone1Name=dyn_lights_bott_l_02, bone2Name=fuselage, swing1LimitDegrees=45, swing2LimitDegrees=45, twistLimitDegrees=45}
 {bone1Name=dyn_lights_bott_l_01, bone2Name=fuselage, swing1LimitDegrees=45, swing2LimitDegrees=45, twistLimitDegrees=45}
 {bone1Name=dyn_rights_top_r_01, bone2Name=fuselage, swing1LimitDegrees=45, swing2LimitDegrees=45, twistLimitDegrees=45}
 {bone1Name=dyn_rights_top_r_02, bone2Name=fuselage, swing1LimitDegrees=45, swing2LimitDegrees=45, twistLimitDegrees=45}
 {bone1Name=dyn_rights_top_r_03, bone2Name=fuselage, swing1LimitDegrees=45, swing2LimitDegrees=45, twistLimitDegrees=45}
 {bone1Name=dyn_rights_top_r_04, bone2Name=fuselage, swing1LimitDegrees=45, swing2LimitDegrees=45, twistLimitDegrees=45}
 {bone1Name=dyn_rights_top_r_05, bone2Name=fuselage, swing1LimitDegrees=45, swing2LimitDegrees=45, twistLimitDegrees=45}
 {bone1Name=dyn_rights_bott_r_02, bone2Name=fuselage, swing1LimitDegrees=45, swing2LimitDegrees=45, twistLimitDegrees=45}
 {bone1Name=dyn_rights_bott_r_01, bone2Name=fuselage, swing1LimitDegrees=45, swing2LimitDegrees=45, twistLimitDegrees=45}
]
