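# Animation Blueprint for the Santa biplane that adds secondary motion to its string lights, paper, dancer, wings and engine pistons.
# Two RigidBody passes simulate the input pose, then additive bone rotations layer on sine-driven jiggle.
# Three speed-scaled light bones act as sources, copied at varying weights to the other light bones.
# The event graph builds the oscillating rotations and scales wing shake by flight-control torque times velocity.
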AnimBlueprint /Game/Athena/DrivableVehicles/Biplane_Santa/Meshes/Biplane_Santa_AnimBP_Dyn
target_skeleton=/Game/Athena/DrivableVehicles/Biplane/Meshes/Biplane_Skeleton
graphs [
 AnimGraph {nodes [
   AnimGraphNode_SubInput_0 = Input Pose()
   n1 = RigidBody(ComponentPose=Local To Component(LocalPose=@AnimGraphNode_SubInput_0)) {
    bOverrideWorldGravity=true
    lODThreshold=1
    overridePhysicsAsset=/Game/Athena/DrivableVehicles/Biplane/Meshes/Biplane_Phat_Dyn
    alphaBoolBlend {blendOption=Linear}
    componentAppliedLinearAccClamp=<1000 1000 1000>
    componentLinearAccScale=<5 5 5>
    componentLinearVelScale=<1 1 1>
    overrideWorldGravity {z=-980}
   }
   n2 = RigidBody(ComponentPose=@n1) {
    bOverrideWorldGravity=true
    lODThreshold=1
    overridePhysicsAsset=&/Biplane_Santa_Phat_Dyn
    alphaBoolBlend {blendOption=Linear}
    componentAppliedLinearAccClamp=<1000 1000 1000>
    componentLinearAccScale=<5 5 5>
    componentLinearVelScale=<1 1 1>
    overrideWorldGravity {z=-980}
   }
   n3 = Get SpeedABS() -> (SpeedABS:float)
   n4 = "Transform (Modify) Bone - Bone: dyn_lights_top_l_013"!(ComponentPose=@n2, Alpha=@n3, Rotation=Get Random Rotation 03() -> (Random Rotation 03:Rotator)) -> (Pose) {
    rotationMode=BMM_Additive
    rotationSpace=BCS_BoneSpace
    translationSpace=BCS_ParentBoneSpace
    alphaBoolBlend {blendOption=Linear}
    alphaScaleBias {scale=0.2}
    alphaScaleBiasClamp {bMapRange=true, interpSpeedDecreasing=0.1, interpSpeedIncreasing=0.1, inRange {max=60, min=15}}
    boneToModify {boneName=dyn_lights_top_l_013}
   }
   n5 = Get Random Rotation 02() -> (Random Rotation 02:Rotator)
   n6 = "Transform (Modify) Bone - Bone: dyn_lights_top_l_010"!(ComponentPose=@n4, Alpha=@n3, Rotation=@n5) -> (Pose) {
    rotationMode=BMM_Additive
    rotationSpace=BCS_BoneSpace
    translationSpace=BCS_ParentBoneSpace
    alphaBoolBlend {blendOption=Linear}
    alphaScaleBias {scale=0.2}
    alphaScaleBiasClamp {bMapRange=true, interpSpeedDecreasing=0.1, interpSpeedIncreasing=0.1, inRange {max=60, min=15}}
    boneToModify {boneName=dyn_lights_top_l_010}
   }
   n7 = Get Random_Rotation_01() -> (Random_Rotation_01:Rotator)
   n8 = "Transform (Modify) Bone - Bone: dyn_lights_top_l_06"!(ComponentPose=@n6, Alpha=@n3, Rotation=@n7) -> (Pose) {
    rotationMode=BMM_Additive
    rotationSpace=BCS_BoneSpace
    translationSpace=BCS_ParentBoneSpace
    alphaBoolBlend {blendOption=Linear}
    alphaScaleBias {scale=0.2}
    alphaScaleBiasClamp {bMapRange=true, interpSpeedDecreasing=0.1, interpSpeedIncreasing=0.1, inRange {max=60, min=15}}
    boneToModify {boneName=dyn_lights_top_l_06}
   }
   n9 = Copy Bone - Source Bone: dyn_lights_top_l_06 - Target Bone: dyn_lights_top_l_012!(ComponentPose:Component Space Pose Link Structure=@n8, Alpha:float=1) -> (Pose:Component Space Pose Link Structure) {
    is=title AnimGraphNode_CopyBone
    bCopyRotation=true
    controlSpace=BCS_ParentBoneSpace
    alphaBoolBlend {blendOption=Linear}
    sourceBone {boneName=dyn_lights_top_l_06}
    targetBone {boneName=dyn_lights_top_l_012}
   }
   n10 = Copy Bone - Source Bone: dyn_lights_top_l_06 - Target Bone: dyn_lights_top_l_011!(ComponentPose:Component Space Pose Link Structure=@n9, Alpha:float=0.75) -> (Pose:Component Space Pose Link Structure) {
    is=title AnimGraphNode_CopyBone
    bCopyRotation=true
    controlSpace=BCS_ParentBoneSpace
    alphaBoolBlend {blendOption=Linear}
    sourceBone {boneName=dyn_lights_top_l_06}
    targetBone {boneName=dyn_lights_top_l_011}
   }
   n11 = Copy Bone - Source Bone: dyn_lights_top_l_013 - Target Bone: dyn_lights_top_l_09!(ComponentPose:Component Space Pose Link Structure=@n10, Alpha:float=1) -> (Pose:Component Space Pose Link Structure) {
    is=title AnimGraphNode_CopyBone
    bCopyRotation=true
    controlSpace=BCS_ParentBoneSpace
    alphaBoolBlend {blendOption=Linear}
    sourceBone {boneName=dyn_lights_top_l_013}
    targetBone {boneName=dyn_lights_top_l_09}
   }
   n12 = Copy Bone - Source Bone: dyn_lights_top_l_010 - Target Bone: dyn_lights_top_l_08!(ComponentPose:Component Space Pose Link Structure=@n11, Alpha:float=0.5) -> (Pose:Component Space Pose Link Structure) {
    is=title AnimGraphNode_CopyBone
    bCopyRotation=true
    controlSpace=BCS_ParentBoneSpace
    alphaBoolBlend {blendOption=Linear}
    sourceBone {boneName=dyn_lights_top_l_010}
    targetBone {boneName=dyn_lights_top_l_08}
   }
   n13 = Copy Bone - Source Bone: dyn_lights_top_l_013 - Target Bone: dyn_lights_top_l_07!(ComponentPose:Component Space Pose Link Structure=@n12, Alpha:float=0.8) -> (Pose:Component Space Pose Link Structure) {
    is=title AnimGraphNode_CopyBone
    bCopyRotation=true
    controlSpace=BCS_ParentBoneSpace
    alphaBoolBlend {blendOption=Linear}
    sourceBone {boneName=dyn_lights_top_l_013}
    targetBone {boneName=dyn_lights_top_l_07}
   }
   n14 = Copy Bone - Source Bone: dyn_lights_top_l_010 - Target Bone: dyn_lights_top_l_05!(ComponentPose:Component Space Pose Link Structure=@n13, Alpha:float=0.5) -> (Pose:Component Space Pose Link Structure) {
    is=title AnimGraphNode_CopyBone
    bCopyRotation=true
    controlSpace=BCS_ParentBoneSpace
    alphaBoolBlend {blendOption=Linear}
    sourceBone {boneName=dyn_lights_top_l_010}
    targetBone {boneName=dyn_lights_top_l_05}
   }
   n15 = Copy Bone - Source Bone: dyn_lights_top_l_013 - Target Bone: dyn_lights_top_l_04!(ComponentPose:Component Space Pose Link Structure=@n14, Alpha:float=0.5) -> (Pose:Component Space Pose Link Structure) {
    is=title AnimGraphNode_CopyBone
    bCopyRotation=true
    controlSpace=BCS_ParentBoneSpace
    alphaBoolBlend {blendOption=Linear}
    sourceBone {boneName=dyn_lights_top_l_013}
    targetBone {boneName=dyn_lights_top_l_04}
   }
   n16 = Copy Bone - Source Bone: dyn_lights_top_l_06 - Target Bone: dyn_lights_top_l_03!(ComponentPose:Component Space Pose Link Structure=@n15, Alpha:float=0.5) -> (Pose:Component Space Pose Link Structure) {
    is=title AnimGraphNode_CopyBone
    bCopyRotation=true
    controlSpace=BCS_ParentBoneSpace
    alphaBoolBlend {blendOption=Linear}
    sourceBone {boneName=dyn_lights_top_l_06}
    targetBone {boneName=dyn_lights_top_l_03}
   }
   n17 = Copy Bone - Source Bone: dyn_lights_top_l_013 - Target Bone: dyn_lights_top_l_02!(ComponentPose:Component Space Pose Link Structure=@n16, Alpha:float=0.5) -> (Pose:Component Space Pose Link Structure) {
    is=title AnimGraphNode_CopyBone
    bCopyRotation=true
    controlSpace=BCS_ParentBoneSpace
    alphaBoolBlend {blendOption=Linear}
    sourceBone {boneName=dyn_lights_top_l_013}
    targetBone {boneName=dyn_lights_top_l_02}
   }
   n18 = Copy Bone - Source Bone: dyn_lights_top_l_010 - Target Bone: dyn_lights_top_l_01!(ComponentPose:Component Space Pose Link Structure=@n17, Alpha:float=0.25) -> (Pose:Component Space Pose Link Structure) {
    is=title AnimGraphNode_CopyBone
    bCopyRotation=true
    controlSpace=BCS_ParentBoneSpace
    alphaBoolBlend {blendOption=Linear}
    sourceBone {boneName=dyn_lights_top_l_010}
    targetBone {boneName=dyn_lights_top_l_01}
   }
   n19 = Copy Bone - Source Bone: dyn_lights_top_l_013 - Target Bone: dyn_rights_top_r_01!(ComponentPose:Component Space Pose Link Structure=@n18, Alpha:float=1) -> (Pose:Component Space Pose Link Structure) {
    is=title AnimGraphNode_CopyBone
    bCopyRotation=true
    controlSpace=BCS_ParentBoneSpace
    alphaBoolBlend {blendOption=Linear}
    sourceBone {boneName=dyn_lights_top_l_013}
    targetBone {boneName=dyn_rights_top_r_01}
   }
   n20 = Copy Bone - Source Bone: dyn_lights_top_l_06 - Target Bone: dyn_rights_top_r_02!(ComponentPose:Component Space Pose Link Structure=@n19, Alpha:float=1) -> (Pose:Component Space Pose Link Structure) {
    is=title AnimGraphNode_CopyBone
    bCopyRotation=true
    controlSpace=BCS_ParentBoneSpace
    alphaBoolBlend {blendOption=Linear}
    sourceBone {boneName=dyn_lights_top_l_06}
    targetBone {boneName=dyn_rights_top_r_02}
   }
   n21 = Copy Bone - Source Bone: dyn_lights_top_l_013 - Target Bone: dyn_rights_top_r_03!(ComponentPose:Component Space Pose Link Structure=@n20, Alpha:float=1) -> (Pose:Component Space Pose Link Structure) {
    is=title AnimGraphNode_CopyBone
    bCopyRotation=true
    controlSpace=BCS_ParentBoneSpace
    alphaBoolBlend {blendOption=Linear}
    sourceBone {boneName=dyn_lights_top_l_013}
    targetBone {boneName=dyn_rights_top_r_03}
   }
   n22 = Copy Bone - Source Bone: dyn_lights_top_l_010 - Target Bone: dyn_rights_top_r_04!(ComponentPose:Component Space Pose Link Structure=@n21, Alpha:float=0.8) -> (Pose:Component Space Pose Link Structure) {
    is=title AnimGraphNode_CopyBone
    bCopyRotation=true
    controlSpace=BCS_ParentBoneSpace
    alphaBoolBlend {blendOption=Linear}
    sourceBone {boneName=dyn_lights_top_l_010}
    targetBone {boneName=dyn_rights_top_r_04}
   }
   n23 = Copy Bone - Source Bone: dyn_lights_top_l_06 - Target Bone: dyn_rights_top_r_05!(ComponentPose:Component Space Pose Link Structure=@n22, Alpha:float=1) -> (Pose:Component Space Pose Link Structure) {
    is=title AnimGraphNode_CopyBone
    bCopyRotation=true
    controlSpace=BCS_ParentBoneSpace
    alphaBoolBlend {blendOption=Linear}
    sourceBone {boneName=dyn_lights_top_l_06}
    targetBone {boneName=dyn_rights_top_r_05}
   }
   n24 = Copy Bone - Source Bone: dyn_lights_top_l_013 - Target Bone: dyn_rights_top_r_06!(ComponentPose:Component Space Pose Link Structure=@n23, Alpha:float=1) -> (Pose:Component Space Pose Link Structure) {
    is=title AnimGraphNode_CopyBone
    bCopyRotation=true
    controlSpace=BCS_ParentBoneSpace
    alphaBoolBlend {blendOption=Linear}
    sourceBone {boneName=dyn_lights_top_l_013}
    targetBone {boneName=dyn_rights_top_r_06}
   }
   n25 = Copy Bone - Source Bone: dyn_lights_top_l_06 - Target Bone: dyn_rights_top_r_07!(ComponentPose:Component Space Pose Link Structure=@n24, Alpha:float=0.7) -> (Pose:Component Space Pose Link Structure) {
    is=title AnimGraphNode_CopyBone
    bCopyRotation=true
    controlSpace=BCS_ParentBoneSpace
    alphaBoolBlend {blendOption=Linear}
    sourceBone {boneName=dyn_lights_top_l_06}
    targetBone {boneName=dyn_rights_top_r_07}
   }
   n26 = Copy Bone - Source Bone: dyn_lights_top_l_013 - Target Bone: dyn_rights_top_r_08!(ComponentPose:Component Space Pose Link Structure=@n25, Alpha:float=1) -> (Pose:Component Space Pose Link Structure) {
    is=title AnimGraphNode_CopyBone
    bCopyRotation=true
    controlSpace=BCS_ParentBoneSpace
    alphaBoolBlend {blendOption=Linear}
    sourceBone {boneName=dyn_lights_top_l_013}
    targetBone {boneName=dyn_rights_top_r_08}
   }
   n27 = Copy Bone - Source Bone: dyn_lights_top_l_010 - Target Bone: dyn_rights_top_r_09!(ComponentPose:Component Space Pose Link Structure=@n26, Alpha:float=0.5) -> (Pose:Component Space Pose Link Structure) {
    is=title AnimGraphNode_CopyBone
    bCopyRotation=true
    controlSpace=BCS_ParentBoneSpace
    alphaBoolBlend {blendOption=Linear}
    sourceBone {boneName=dyn_lights_top_l_010}
    targetBone {boneName=dyn_rights_top_r_09}
   }
   n28 = Copy Bone - Source Bone: dyn_lights_top_l_013 - Target Bone: dyn_rights_top_r_010!(ComponentPose:Component Space Pose Link Structure=@n27, Alpha:float=1) -> (Pose:Component Space Pose Link Structure) {
    is=title AnimGraphNode_CopyBone
    bCopyRotation=true
    controlSpace=BCS_ParentBoneSpace
    alphaBoolBlend {blendOption=Linear}
    sourceBone {boneName=dyn_lights_top_l_013}
    targetBone {boneName=dyn_rights_top_r_010}
   }
   n29 = Copy Bone - Source Bone: dyn_lights_top_l_06 - Target Bone: dyn_rights_top_r_011!(ComponentPose:Component Space Pose Link Structure=@n28, Alpha:float=1) -> (Pose:Component Space Pose Link Structure) {
    is=title AnimGraphNode_CopyBone
    bCopyRotation=true
    controlSpace=BCS_ParentBoneSpace
    alphaBoolBlend {blendOption=Linear}
    sourceBone {boneName=dyn_lights_top_l_06}
    targetBone {boneName=dyn_rights_top_r_011}
   }
   n30 = Copy Bone - Source Bone: dyn_lights_top_l_010 - Target Bone: dyn_rights_top_r_012!(ComponentPose:Component Space Pose Link Structure=@n29, Alpha:float=1) -> (Pose:Component Space Pose Link Structure) {
    is=title AnimGraphNode_CopyBone
    bCopyRotation=true
    controlSpace=BCS_ParentBoneSpace
    alphaBoolBlend {blendOption=Linear}
    sourceBone {boneName=dyn_lights_top_l_010}
    targetBone {boneName=dyn_rights_top_r_012}
   }
   n31 = Copy Bone - Source Bone: dyn_lights_top_l_013 - Target Bone: dyn_rights_top_r_013!(ComponentPose:Component Space Pose Link Structure=@n30, Alpha:float=2) -> (Pose:Component Space Pose Link Structure) {
    is=title AnimGraphNode_CopyBone
    bCopyRotation=true
    controlSpace=BCS_ParentBoneSpace
    alphaBoolBlend {blendOption=Linear}
    sourceBone {boneName=dyn_lights_top_l_013}
    targetBone {boneName=dyn_rights_top_r_013}
   }
   n32 = Copy Bone - Source Bone: dyn_lights_top_l_013 - Target Bone: dyn_lights_bott_l_09!(ComponentPose:Component Space Pose Link Structure=@n31, Alpha:float=0.8) -> (Pose:Component Space Pose Link Structure) {
    is=title AnimGraphNode_CopyBone
    bCopyRotation=true
    controlSpace=BCS_ParentBoneSpace
    alphaBoolBlend {blendOption=Linear}
    sourceBone {boneName=dyn_lights_top_l_013}
    targetBone {boneName=dyn_lights_bott_l_09}
   }
   n33 = Copy Bone - Source Bone: dyn_lights_top_l_010 - Target Bone: dyn_lights_bott_l_08!(ComponentPose:Component Space Pose Link Structure=@n32, Alpha:float=1) -> (Pose:Component Space Pose Link Structure) {
    is=title AnimGraphNode_CopyBone
    bCopyRotation=true
    controlSpace=BCS_ParentBoneSpace
    alphaBoolBlend {blendOption=Linear}
    sourceBone {boneName=dyn_lights_top_l_010}
    targetBone {boneName=dyn_lights_bott_l_08}
   }
   n34 = Copy Bone - Source Bone: dyn_lights_top_l_06 - Target Bone: dyn_lights_bott_l_07!(ComponentPose:Component Space Pose Link Structure=@n33, Alpha:float=0.5) -> (Pose:Component Space Pose Link Structure) {
    is=title AnimGraphNode_CopyBone
    bCopyRotation=true
    controlSpace=BCS_ParentBoneSpace
    alphaBoolBlend {blendOption=Linear}
    sourceBone {boneName=dyn_lights_top_l_06}
    targetBone {boneName=dyn_lights_bott_l_07}
   }
   n35 = Copy Bone - Source Bone: dyn_lights_top_l_013 - Target Bone: dyn_lights_bott_l_06!(ComponentPose:Component Space Pose Link Structure=@n34, Alpha:float=0.75) -> (Pose:Component Space Pose Link Structure) {
    is=title AnimGraphNode_CopyBone
    bCopyRotation=true
    controlSpace=BCS_ParentBoneSpace
    alphaBoolBlend {blendOption=Linear}
    sourceBone {boneName=dyn_lights_top_l_013}
    targetBone {boneName=dyn_lights_bott_l_06}
   }
   n36 = Copy Bone - Source Bone: dyn_lights_top_l_010 - Target Bone: dyn_lights_bott_l_05!(ComponentPose:Component Space Pose Link Structure=@n35, Alpha:float=1) -> (Pose:Component Space Pose Link Structure) {
    is=title AnimGraphNode_CopyBone
    bCopyRotation=true
    controlSpace=BCS_ParentBoneSpace
    alphaBoolBlend {blendOption=Linear}
    sourceBone {boneName=dyn_lights_top_l_010}
    targetBone {boneName=dyn_lights_bott_l_05}
   }
   n37 = Copy Bone - Source Bone: dyn_lights_top_l_06 - Target Bone: dyn_lights_bott_l_04!(ComponentPose:Component Space Pose Link Structure=@n36, Alpha:float=0.9) -> (Pose:Component Space Pose Link Structure) {
    is=title AnimGraphNode_CopyBone
    bCopyRotation=true
    controlSpace=BCS_ParentBoneSpace
    alphaBoolBlend {blendOption=Linear}
    sourceBone {boneName=dyn_lights_top_l_06}
    targetBone {boneName=dyn_lights_bott_l_04}
   }
   n38 = Copy Bone - Source Bone: dyn_lights_top_l_010 - Target Bone: dyn_lights_bott_l_03!(ComponentPose:Component Space Pose Link Structure=@n37, Alpha:float=1) -> (Pose:Component Space Pose Link Structure) {
    is=title AnimGraphNode_CopyBone
    bCopyRotation=true
    controlSpace=BCS_ParentBoneSpace
    alphaBoolBlend {blendOption=Linear}
    sourceBone {boneName=dyn_lights_top_l_010}
    targetBone {boneName=dyn_lights_bott_l_03}
   }
   n39 = Copy Bone - Source Bone: dyn_lights_top_l_013 - Target Bone: dyn_lights_bott_l_02!(ComponentPose:Component Space Pose Link Structure=@n38, Alpha:float=0.6) -> (Pose:Component Space Pose Link Structure) {
    is=title AnimGraphNode_CopyBone
    bCopyRotation=true
    controlSpace=BCS_ParentBoneSpace
    alphaBoolBlend {blendOption=Linear}
    sourceBone {boneName=dyn_lights_top_l_013}
    targetBone {boneName=dyn_lights_bott_l_02}
   }
   n40 = Copy Bone - Source Bone: dyn_lights_top_l_06 - Target Bone: dyn_lights_bott_l_01!(ComponentPose:Component Space Pose Link Structure=@n39, Alpha:float=1) -> (Pose:Component Space Pose Link Structure) {
    is=title AnimGraphNode_CopyBone
    bCopyRotation=true
    controlSpace=BCS_ParentBoneSpace
    alphaBoolBlend {blendOption=Linear}
    sourceBone {boneName=dyn_lights_top_l_06}
    targetBone {boneName=dyn_lights_bott_l_01}
   }
   n41 = Copy Bone - Source Bone: dyn_lights_top_l_010 - Target Bone: dyn_rights_bott_r_01!(ComponentPose:Component Space Pose Link Structure=@n40, Alpha:float=1) -> (Pose:Component Space Pose Link Structure) {
    is=title AnimGraphNode_CopyBone
    bCopyRotation=true
    controlSpace=BCS_ParentBoneSpace
    alphaBoolBlend {blendOption=Linear}
    sourceBone {boneName=dyn_lights_top_l_010}
    targetBone {boneName=dyn_rights_bott_r_01}
   }
   n42 = Copy Bone - Source Bone: dyn_lights_top_l_06 - Target Bone: dyn_rights_bott_r_02!(ComponentPose:Component Space Pose Link Structure=@n41, Alpha:float=0.9) -> (Pose:Component Space Pose Link Structure) {
    is=title AnimGraphNode_CopyBone
    bCopyRotation=true
    controlSpace=BCS_ParentBoneSpace
    alphaBoolBlend {blendOption=Linear}
    sourceBone {boneName=dyn_lights_top_l_06}
    targetBone {boneName=dyn_rights_bott_r_02}
   }
   n43 = Copy Bone - Source Bone: dyn_lights_top_l_013 - Target Bone: dyn_rights_bott_r_03!(ComponentPose:Component Space Pose Link Structure=@n42, Alpha:float=1) -> (Pose:Component Space Pose Link Structure) {
    is=title AnimGraphNode_CopyBone
    bCopyRotation=true
    controlSpace=BCS_ParentBoneSpace
    alphaBoolBlend {blendOption=Linear}
    sourceBone {boneName=dyn_lights_top_l_013}
    targetBone {boneName=dyn_rights_bott_r_03}
   }
   n44 = Copy Bone - Source Bone: dyn_lights_top_l_010 - Target Bone: dyn_rights_bott_r_04!(ComponentPose:Component Space Pose Link Structure=@n43, Alpha:float=0.78) -> (Pose:Component Space Pose Link Structure) {
    is=title AnimGraphNode_CopyBone
    bCopyRotation=true
    controlSpace=BCS_ParentBoneSpace
    alphaBoolBlend {blendOption=Linear}
    sourceBone {boneName=dyn_lights_top_l_010}
    targetBone {boneName=dyn_rights_bott_r_04}
   }
   n45 = Copy Bone - Source Bone: dyn_lights_top_l_06 - Target Bone: dyn_rights_bott_r_05!(ComponentPose:Component Space Pose Link Structure=@n44, Alpha:float=1) -> (Pose:Component Space Pose Link Structure) {
    is=title AnimGraphNode_CopyBone
    bCopyRotation=true
    controlSpace=BCS_ParentBoneSpace
    alphaBoolBlend {blendOption=Linear}
    sourceBone {boneName=dyn_lights_top_l_06}
    targetBone {boneName=dyn_rights_bott_r_05}
   }
   n46 = Copy Bone - Source Bone: dyn_lights_top_l_013 - Target Bone: dyn_rights_bott_r_06!(ComponentPose:Component Space Pose Link Structure=@n45, Alpha:float=1) -> (Pose:Component Space Pose Link Structure) {
    is=title AnimGraphNode_CopyBone
    bCopyRotation=true
    controlSpace=BCS_ParentBoneSpace
    alphaBoolBlend {blendOption=Linear}
    sourceBone {boneName=dyn_lights_top_l_013}
    targetBone {boneName=dyn_rights_bott_r_06}
   }
   n47 = Copy Bone - Source Bone: dyn_lights_top_l_06 - Target Bone: dyn_rights_bott_r_07!(ComponentPose:Component Space Pose Link Structure=@n46, Alpha:float=0.5) -> (Pose:Component Space Pose Link Structure) {
    is=title AnimGraphNode_CopyBone
    bCopyRotation=true
    controlSpace=BCS_ParentBoneSpace
    alphaBoolBlend {blendOption=Linear}
    sourceBone {boneName=dyn_lights_top_l_06}
    targetBone {boneName=dyn_rights_bott_r_07}
   }
   n48 = Copy Bone - Source Bone: dyn_lights_top_l_010 - Target Bone: dyn_rights_bott_r_08!(ComponentPose:Component Space Pose Link Structure=@n47, Alpha:float=1) -> (Pose:Component Space Pose Link Structure) {
    is=title AnimGraphNode_CopyBone
    bCopyRotation=true
    controlSpace=BCS_ParentBoneSpace
    alphaBoolBlend {blendOption=Linear}
    sourceBone {boneName=dyn_lights_top_l_010}
    targetBone {boneName=dyn_rights_bott_r_08}
   }
   n49 = Copy Bone - Source Bone: dyn_lights_top_l_013 - Target Bone: dyn_rights_bott_r_09!(ComponentPose:Component Space Pose Link Structure=@n48, Alpha:float=0.88) -> (Pose:Component Space Pose Link Structure) {
    is=title AnimGraphNode_CopyBone
    bCopyRotation=true
    controlSpace=BCS_ParentBoneSpace
    alphaBoolBlend {blendOption=Linear}
    sourceBone {boneName=dyn_lights_top_l_013}
    targetBone {boneName=dyn_rights_bott_r_09}
   }
   n50 = "Transform (Modify) Bone - Bone: paper"!(ComponentPose=@n49, Alpha=@n3, Rotation=@n7) -> (Pose) {
    rotationMode=BMM_Additive
    rotationSpace=BCS_BoneSpace
    translationSpace=BCS_ParentBoneSpace
    alphaBoolBlend {blendOption=Linear}
    alphaScaleBias {scale=0.2}
    alphaScaleBiasClamp {bMapRange=true, interpSpeedDecreasing=0.1, interpSpeedIncreasing=0.1, inRange {max=60, min=15}}
    boneToModify {boneName=paper}
   }
   n51 = "Transform (Modify) Bone - Bone: dancer_base"!(ComponentPose=@n50, Alpha=@n3, Rotation=@n5) -> (Pose) {
    rotationMode=BMM_Additive
    rotationSpace=BCS_BoneSpace
    translationSpace=BCS_ParentBoneSpace
    alphaBoolBlend {blendOption=Linear}
    alphaScaleBias {scale=0.2}
    alphaScaleBiasClamp {bMapRange=true, interpSpeedDecreasing=0.1, interpSpeedIncreasing=0.1, inRange {max=60, min=15}}
    boneToModify {boneName=dancer_base}
   }
   n52 = Get WingRotations() -> (WingRotations:Rotator)
   n53 = Get WingJiggleAlpha() -> (WingJiggleAlpha:double)
   n54 = "Transform (Modify) Bone - Bone: wing_top_l"!(ComponentPose=@n51, Alpha=@n53, Rotation=@n52) -> (Pose) {
    rotationMode=BMM_Additive
    rotationSpace=BCS_BoneSpace
    translationSpace=BCS_ParentBoneSpace
    alphaBoolBlend {blendOption=Linear}
    boneToModify {boneName=wing_top_l}
   }
   n55 = "Transform (Modify) Bone - Bone: wing_bott_l"!(ComponentPose=@n54, Alpha=@n53, Rotation=@n52) -> (Pose) {
    rotationMode=BMM_Additive
    rotationSpace=BCS_BoneSpace
    translationSpace=BCS_ParentBoneSpace
    alphaBoolBlend {blendOption=Linear}
    boneToModify {boneName=wing_bott_l}
   }
   n56 = "Transform (Modify) Bone - Bone: wing_top_r"!(ComponentPose=@n55, Alpha=@n53, Rotation=@n52) -> (Pose) {
    rotationMode=BMM_Additive
    rotationSpace=BCS_BoneSpace
    translationSpace=BCS_ParentBoneSpace
    alphaBoolBlend {blendOption=Linear}
    boneToModify {boneName=wing_top_r}
   }
   n57 = "Transform (Modify) Bone - Bone: wing_bott_r"!(ComponentPose=@n56, Alpha=@n53, Rotation=@n52) -> (Pose) {
    rotationMode=BMM_Additive
    rotationSpace=BCS_BoneSpace
    translationSpace=BCS_ParentBoneSpace
    alphaBoolBlend {blendOption=Linear}
    boneToModify {boneName=wing_bott_r}
   }
   n58 = Get CurrentDriveJiggleRandom2() -> (CurrentDriveJiggleRandom2:Rotator)
   n59 = "Transform (Modify) Bone - Bone: engine_piston_fr_l"(ComponentPose=@n57, Alpha=@n3, Translation="0,0,0", Rotation=@n58, Scale="1,1,1") {
    lODThreshold=1
    rotationMode=BMM_Additive
    rotationSpace=BCS_BoneSpace
    alphaBoolBlend {blendOption=Linear}
    alphaScaleBiasClamp {bMapRange=true, scale=0.9, inRange {max=45}, outRange {min=0.2}}
    boneToModify {boneName=engine_piston_fr_l}
   }
   n60 = Get CurrentDriveJiggleRandom() -> (CurrentDriveJiggleRandom:Rotator)
   n61 = "Transform (Modify) Bone - Bone: engine_piston_mid_l"(ComponentPose=@n59, Alpha=@n3, Translation="0,0,0", Rotation=@n60, Scale="1,1,1") {
    lODThreshold=1
    rotationMode=BMM_Additive
    rotationSpace=BCS_BoneSpace
    alphaBoolBlend {blendOption=Linear}
    alphaScaleBiasClamp {bMapRange=true, scale=1.1, inRange {max=65}, outRange {min=0.2}}
    boneToModify {boneName=engine_piston_mid_l}
   }
   n62 = Get CurrentDriveJiggle() -> (CurrentDriveJiggle:Rotator)
   n63 = "Transform (Modify) Bone - Bone: engine_piston_bk_l"(ComponentPose=@n61, Alpha=@n3, Translation="0,0,0", Rotation=@n62, Scale="1,1,1") {
    lODThreshold=1
    rotationMode=BMM_Additive
    rotationSpace=BCS_BoneSpace
    alphaBoolBlend {blendOption=Linear}
    alphaScaleBiasClamp {bMapRange=true, inRange {max=60}, outRange {min=0.2}}
    boneToModify {boneName=engine_piston_bk_l}
   }
   n64 = "Transform (Modify) Bone - Bone: engine_piston_fr_r"(ComponentPose=@n63, Alpha=@n3, Translation="0,0,0", Rotation=@n62, Scale="1,1,1") {
    lODThreshold=1
    rotationMode=BMM_Additive
    rotationSpace=BCS_BoneSpace
    alphaBoolBlend {blendOption=Linear}
    alphaScaleBiasClamp {bMapRange=true, inRange {max=60}, outRange {min=0.2}}
    boneToModify {boneName=engine_piston_fr_r}
   }
   n65 = "Transform (Modify) Bone - Bone: engine_piston_mid_r"(ComponentPose=@n64, Alpha=@n3, Translation="0,0,0", Rotation=@n58, Scale="1,1,1") {
    lODThreshold=1
    rotationMode=BMM_Additive
    rotationSpace=BCS_BoneSpace
    alphaBoolBlend {blendOption=Linear}
    alphaScaleBiasClamp {bMapRange=true, scale=0.9, inRange {max=45}, outRange {min=0.2}}
    boneToModify {boneName=engine_piston_mid_r}
   }
   n66 = "Transform (Modify) Bone - Bone: engine_piston_bk_r"(ComponentPose=@n65, Alpha=@n3, Translation="0,0,0", Rotation=@n60, Scale="1,1,1") {
    lODThreshold=1
    rotationMode=BMM_Additive
    rotationSpace=BCS_BoneSpace
    alphaBoolBlend {blendOption=Linear}
    alphaScaleBiasClamp {bMapRange=true, scale=1.1, inRange {max=65}, outRange {min=0.2}}
    boneToModify {boneName=engine_piston_bk_r}
   }
   Output Pose(Result=Component To Local(ComponentPose=@n66)) {layerGroup=None}
  ]}
 EventGraph {nodes [
   Event BlueprintUpdateAnimation()
   n1 = GetGameTimeInSeconds()
   n2 = float * float(A=1, B=45)
   n3 = float * float(A="Sin (Radians)"(A=float * float(A=@n2, B=@n1)), B=15)
   n4 = float * float(A=1, B=20)
   n5 = float * float(A="Sin (Radians)"(A=float * float(A=@n4, B=@n1)), B=30)
   n6 = float * float(A=1, B=25)
   n7 = float * float(A="Sin (Radians)"(A=float * float(A=@n6, B=@n1)), B=10)
   > Set Random_Rotation_01(Random_Rotation_01:Rotator=MakeRotator(Roll=@n3, Pitch=@n5, Yaw=@n7)) -> (Output_Get:Rotator)
   > Set Random Rotation 02(Random Rotation 02:Rotator=MakeRotator(Roll=@n7, Pitch=@n5, Yaw=@n3)) -> (Output_Get:Rotator)
   > Set Random Rotation 03(Random Rotation 03:Rotator=MakeRotator(Roll=@n5, Pitch=@n3, Yaw=@n7)) -> (Output_Get:Rotator)
   n8 = Get CurrentDriveJiggle()
   n9 = BreakRotator(InRot=@n8)
   n10 = GetGameTimeInSeconds()
   n11 = float * float(A=float * float(A=3, B=@n9.Pitch), B=@n10)
   n12 = float * float(A=float * float(A=2, B=@n9.Yaw), B=@n10)
   > Set CurrentDriveJiggleRandom(CurrentDriveJiggleRandom:Rotator=MakeRotator(Roll=@n9, Pitch="Sin (Radians)"(A=@n11), Yaw="Sin (Radians)"(A=@n12))) -> (Output_Get:Rotator)
   n13 = BreakRotator(InRot=@n8)
   n14 = float * float(A=float * float(A=0.5, B=@n13.Pitch), B=@n10)
   n15 = float * float(A=float * float(A=0.7, B=@n13.Yaw), B=@n10)
   > n16 = Set CurrentDriveJiggleRandom2(CurrentDriveJiggleRandom2:Rotator=MakeRotator(Roll=@n13, Pitch="Sin (Radians)"(A=@n14), Yaw="Sin (Radians)"(A=@n15))) -> (Output_Get:Rotator)
   n17 = GetGameTimeInSeconds()
   n18 = Set random_stream(execute=@n27.then, random_stream:Random Stream Structure=MakeRandomStream!(InitialSeed:int=0) -> (ReturnValue:Random Stream Structure)) -> (Output_Get:Random Stream Structure)
   n19 = RandomFloatInRangeFromStream!(Stream:"Random Stream Structure (by ref)"=@n18, Min:float=32, Max:float=50) -> (ReturnValue:float)
   n20 = "Sin (Radians)"(A=float * float(A=float * float(A=1, B=@n19), B=@n17))
   n21 = "Sin (Radians)"(A=float * float(A=float * float(A=1, B=@n19), B=@n17))
   n22 = RandomFloatInRangeFromStream(Stream=@n18, Min=10, Max=20)
   n23 = "Sin (Radians)"(A=float * float(A=float * float(A=1, B=@n22), B=@n17))
   Set WingRotations(
    execute=@n16.then
    WingRotations:Rotator=MakeRotator(Roll=float * float(A=@n23, B=1), Pitch=float * float(A=@n21, B=0.75), Yaw=float * float(A=@n20, B=0.6))
   ) -> (Output_Get:Rotator)
   > n24 = Cast To FerretVehicle!(execute, Object=GetOwningActor()) -> (then, CastFailed, AsFerret Vehicle:Ferret Vehicle Object Reference) {is=title K2Node_DynamicCast}
   n25 = Vector Length(A=GetFlightControlTorque!(self:Fort Athena Doghouse Vehicle Object Reference=@n24) -> (ReturnValue:Vector))
   n26 = Vector Length(A=GetVelocity!(self:Actor Object Reference=@n24) -> (ReturnValue:Vector))
   > Set WingJiggleAlpha(WingJiggleAlpha:double=MapRangeClamped(Value=float * float(A=@n25, B=@n26), InRangeA=10000, InRangeB=500000, OutRangeB=0.5)) -> (Output_Get:double)
   n27 = Event BlueprintInitializeAnimation()
   Sequence()
   > n28 = Branch(Condition=Get bIsBoosting() -> (bIsBoosting:bool))
   > Do Once(Reset=@n28.else)
   n29 = GetGameTimeInSeconds()
   > Set WingJiggleAlpha(WingJiggleAlpha=float + float(A=0, B=FInterpTo(Current=0, Target=0.5, DeltaTime=@n29, InterpSpeed=1)) {is=title K2Node_CommutativeAssociativeBinaryOperator})
   MapRangeClamped(
    Value=Break Fort Recharging Action Timer!(FortRechargingActionTimer:"Fort Recharging Action Timer Structure (by ref)"=Get BoostAction!(self:Fort Athena Doghouse Vehicle Object Reference=@n24) -> (BoostAction:Fort Recharging Action Timer Structure)) -> (
     ChargeRate:float=~
     ActiveExpenseRate:float=~
     PassiveExpenseRate:float=~
     MinActiveDuration:float=~
     MinActivationCharge:float=~
     ActiveCooldownTime:float=~
     ChargeThreshold:float=~
     Charge:float=~
     bIsActive:bool=~
     bIsCharging:bool=~
     bIsPassive:bool=~
    ) {is=title K2Node_BreakStruct}.Charge
    InRangeA=0
    InRangeB=1
    OutRangeA=1
    OutRangeB=0
   )
  ]}
]
variables [
 CurrentDriveJiggleRandom {type=Rotator}
 CurrentDriveJiggleRandom2 {type=Rotator}
 Random Rotation 02 {type=Rotator}
 Random Rotation 03 {type=Rotator}
 Random_Rotation_01 {type=Rotator}
 random_stream {type=RandomStream, default {initialSeed=0, seed=0}}
 WingJiggleAlpha {type=number}
 WingRotations {type=Rotator}
]
